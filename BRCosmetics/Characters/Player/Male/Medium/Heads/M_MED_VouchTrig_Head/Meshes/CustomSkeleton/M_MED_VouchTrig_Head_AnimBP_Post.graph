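# Post AnimBlueprint for the VouchTrig male bird head cosmetic, adding rig-driven correction and feather motion to the input pose.
# A Control Rig feeds a fixed additive yaw on head_bird, then a position constraint ties faceAttach to that bone.
# Five chained, partially blended AnimDynamics nodes add spring motion to front and back head feathers.
# The EventGraph remaps AimPitch into headRot_add_1, which the AnimGraph does not visibly read.
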AnimBlueprint /BRCosmetics/Characters/Player/Male/Medium/Heads/M_MED_VouchTrig_Head/Meshes/CustomSkeleton/M_MED_VouchTrig_Head_AnimBP_Post
target_skeleton=&/Fortnite_Base_Head_VouchTrig_Skel
graphs [
 AnimGraph {nodes [
   n1 = Control Rig(Source=Input Pose()) {controlRigAssetReference {blueprintRigClass=&/M_MED_VouchTrig_Head_CtrlRig.M_MED_VouchTrig_Head_CtrlRig_C}}
   n2 = "Transform (Modify) Bone - Bone: head_bird"!(ComponentPose=Local To Component(LocalPose=@n1), bAlphaBoolEnabled:bool=True, Rotation="0,-10,0") -> (Pose) {
    alphaInputType=Bool
    rotationMode=BMM_Additive
    rotationSpace=BCS_BoneSpace
    alphaBoolBlend {blendInTime=0.3, blendOption=ExpOut, blendOutTime=0.3}
    boneToModify {boneName=head_bird}
    rotation {yaw=-10}
   }
   n3 = "Constraint - faceAttach (P)"(ComponentPose=@n2, Alpha=1, ConstraintWeights_0=1) {boneToModify {boneName=faceAttach}, constraintSetup [{targetBone {boneName=head_bird}}], constraintWeights [1]}
   n4 = Anim Dynamics - Bone: dyn_headFeather_1_front!(ComponentPose=@n3, ExternalForce="0,0,0") -> (Pose) {
    alpha=0.6
    angularDampingOverride=1
    angularSpringConstant=15000
    bAngularSpring=true
    bOverrideAngularDamping=true
    gravityScale=0
    simulationSpace=RootRelative
    boundBone {boneName=dyn_headFeather_1_front}
    componentAppliedLinearAccClamp=<3000 3000 3000>
    componentLinearAccScale=<0.4 0.4 0.4>
    componentLinearVelScale=<0.4 0.4 0.4>
    physicsBodyDefinitions [{
      boundBone {boneName=dyn_headFeather_1_front}
      boxExtents=<30 30 30>
      constraintSetup {angularLimitsMax=<0 10 10>, angularLimitsMin=<0 -10 -10>, angularTarget=<1 0 0>}
      localJointOffset=<4 0 0>
     }]
    simSpaceSettings {maxAngularAcceleration=35, maxAngularVelocity=8, simSpaceAngularAlpha=1}
   }
   n5 = Anim Dynamics - Bone: dyn_headFeather_2_front!(ComponentPose=@n4, ExternalForce="0,0,0") -> (Pose) {
    alpha=0.6
    angularDampingOverride=1
    angularSpringConstant=15000
    bAngularSpring=true
    bOverrideAngularDamping=true
    gravityScale=0
    simulationSpace=RootRelative
    boundBone {boneName=dyn_headFeather_2_front}
    componentAppliedLinearAccClamp=<3000 3000 3000>
    componentLinearAccScale=<0.4 0.4 0.4>
    componentLinearVelScale=<0.4 0.4 0.4>
    physicsBodyDefinitions [{
      boundBone {boneName=dyn_headFeather_2_front}
      boxExtents=<30 30 30>
      constraintSetup {angularLimitsMax=<0 10 10>, angularLimitsMin=<0 -10 -10>, angularTarget=<1 0 0>}
      localJointOffset=<4 0 0>
     }]
    simSpaceSettings {maxAngularAcceleration=35, maxAngularVelocity=8, simSpaceAngularAlpha=1}
   }
   n6 = Anim Dynamics - Bone: dyn_headFeather_3_front!(ComponentPose=@n5, ExternalForce="0,0,0") -> (Pose) {
    alpha=0.6
    angularDampingOverride=1
    angularSpringConstant=15000
    bAngularSpring=true
    bOverrideAngularDamping=true
    gravityScale=0
    simulationSpace=RootRelative
    boundBone {boneName=dyn_headFeather_3_front}
    componentAppliedLinearAccClamp=<3000 3000 3000>
    componentLinearAccScale=<0.4 0.4 0.4>
    componentLinearVelScale=<0.4 0.4 0.4>
    physicsBodyDefinitions [{
      boundBone {boneName=dyn_headFeather_3_front}
      boxExtents=<30 30 30>
      constraintSetup {angularLimitsMax=<0 10 10>, angularLimitsMin=<0 -10 -10>, angularTarget=<1 0 0>}
      localJointOffset=<4 0 0>
     }]
    simSpaceSettings {maxAngularAcceleration=35, maxAngularVelocity=8, simSpaceAngularAlpha=1}
   }
   n7 = Anim Dynamics - Bone: dyn_headFeather_1_back!(ComponentPose=@n6, ExternalForce="0,0,0") -> (Pose) {
    alpha=0.6
    angularDampingOverride=1
    angularSpringConstant=15000
    bAngularSpring=true
    bOverrideAngularDamping=true
    gravityScale=0
    simulationSpace=RootRelative
    boundBone {boneName=dyn_headFeather_1_back}
    componentAppliedLinearAccClamp=<3000 3000 3000>
    componentLinearAccScale=<0.4 0.4 0.4>
    componentLinearVelScale=<0.4 0.4 0.4>
    physicsBodyDefinitions [{
      boundBone {boneName=dyn_headFeather_1_back}
      boxExtents=<30 30 30>
      constraintSetup {angularLimitsMax=<0 10 10>, angularLimitsMin=<0 -10 -10>, angularTarget=<1 0 0>}
      localJointOffset=<4 0 0>
     }]
    simSpaceSettings {maxAngularAcceleration=35, maxAngularVelocity=8, simSpaceAngularAlpha=1}
   }
   n8 = Anim Dynamics - Bone: dyn_headFeather_2_back!(ComponentPose=@n7, ExternalForce="0,0,0") -> (Pose) {
    alpha=0.6
    angularDampingOverride=1
    angularSpringConstant=15000
    bAngularSpring=true
    bOverrideAngularDamping=true
    gravityScale=0
    simulationSpace=RootRelative
    boundBone {boneName=dyn_headFeather_2_back}
    componentAppliedLinearAccClamp=<3000 3000 3000>
    componentLinearAccScale=<0.4 0.4 0.4>
    componentLinearVelScale=<0.4 0.4 0.4>
    physicsBodyDefinitions [{
      boundBone {boneName=dyn_headFeather_2_back}
      boxExtents=<30 30 30>
      constraintSetup {angularLimitsMax=<0 10 10>, angularLimitsMin=<0 -10 -10>, angularTarget=<1 0 0>}
      localJointOffset=<4 0 0>
     }]
    simSpaceSettings {maxAngularAcceleration=35, maxAngularVelocity=8, simSpaceAngularAlpha=1}
   }
   Output Pose(Result=Component To Local(ComponentPose=@n8))
  ]}
 EventGraph {nodes [
   Event BlueprintUpdateAnimation()
   n1 = Get AimPitch()
   > Set headRot_add_1(headRot_add_1:Rotator=MakeRotator(Yaw=Remap Float(Input=@n1, OldRange=(X=-10,Y=-40), NewRange=(X=-8,Y=0), Clamp=true) -> (Output=~))) -> (Output_Get:Rotator)
  ]}
]
variables [headRot_add_1 {type=Rotator}]
